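# Light function material for PistolCanary_Countdown that builds animated cloud-noise masks to modulate a light.
# One branch multiplies two rotating noise samples and applies CheapContrast, with UseMPC switching rotation speeds between collection and material parameters.
# A second branch multiplies two panned noise samples, with Panner1-Tiling feeding Time instead of scaling UVs.
# The file shows neither branch connected to a material output.
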Material /PistolCanary_Countdown/Assets/FX/LightFunctions/M_PistolCanary_LightFunction
graph {
 properties {bAutomaticallySetUsageInEditor=false, materialDomain=MD_LightFunction, usedShadingModels=Default Lit}
 nodes [
  n1 = CollectionParameter() {collection=&/MPC_PIstolCanary_LightFunction, parameterName=Texture2 - Speed}
  n2 = ScalarParameter() {defaultValue=250, parameterName=Texture 2 - Speed}
  n3 = MaterialFunctionCall(
   UVs=Multiply(A=TextureCoordinate(), B=ScalarParameter() {defaultValue=3, parameterName=Texture2 - Size})
   "Rotation Angle (0-1)"=Divide(A=Time(), B=StaticSwitchParameter(True=@n1, False=@n2) {parameterName=UseMPC?}) {constB=250}
  ) {materialFunction=&CustomRotator}
  n4 = CollectionParameter() {collection=&MPC_PIstolCanary_LightFunction, parameterName=Texture1 - Speed}
  n5 = ScalarParameter() {defaultValue=-250, parameterName=Texture 1 - Speed}
  n6 = MaterialFunctionCall(
   UVs=Multiply(A=TextureCoordinate(), B=ScalarParameter() {defaultValue=3, parameterName=Texture1 - Size})
   "Rotation Angle (0-1)"=Divide(A=Time(), B=StaticSwitchParameter(True=@n4, False=@n5) {parameterName=UseMPC?}) {constB=-350}
  ) {materialFunction=&CustomRotator}
  MaterialFunctionCall(
   In=Multiply(
    A=TextureSample(UVs=@n3.Rotated Values) {texture=&/PistolCanary_NoiseClouds_Countdown}.R
    B=TextureSample(UVs=@n6.Rotated Values) {texture=&PistolCanary_NoiseClouds_Countdown}.R
   )
   Contrast=ScalarParameter() {parameterName=Contrast}
  ) {materialFunction=&CheapContrast}
  n7 = Multiply(A=TextureCoordinate(), B=ScalarParameter() {defaultValue=1, group=Panner2, parameterName=Panner2-Tiling})
  n8 = VectorParameter() {group=Panner2, parameterName=Panner2, defaultValue=<rgba 0 0 0 1>}
  n9 = AppendVector(A=@n8.R, B=@n8.R)
  n10 = TextureCoordinate()
  n11 = ScalarParameter() {defaultValue=1, group=Panner1, parameterName=Panner1-Tiling}
  n12 = VectorParameter() {group=Panner1, parameterName=Panner1, defaultValue=<rgba 0 0 0 1>}
  n13 = AppendVector(A=@n12.R, B=@n12.R)
  Multiply(
   A=TextureSample(UVs=Panner(Coordinate=@n7, Speed=@n9) {speedX=1}) {texture=&PistolCanary_NoiseClouds_Countdown}.R
   B=TextureSample(UVs=Panner(Coordinate=@n10, Time=@n11, Speed=@n13) {speedX=1}) {texture=&PistolCanary_NoiseClouds_Countdown}.R
  )
 ]
}
